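; Neon City store front awning Blueprint that can spawn extra decorative meshes on its static mesh.
; The construction script filters sockets by name, then adds a hierarchical instanced mesh component at each, with random yaw and uniform scale.
; Functions toggle these add-on meshes and cycle their material through a list.
; It registers the mesh and material lists with Fortnite's exposed asset provider to allow those references.
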(fn AddNewInstancedMesh
  (params (InSocketName :type Name))
  (bind _returnvalue_location (Math|Transform|BreakTransform (Transformation|GetSocketTransform (Variables|Debug|GetStaticMeshComponent) InSocketName "RTS_Actor")))
  (if (Utilities|Array|IsNotEmpty (Variables|Awning|GetInstMeshList))
    (bind _returnvalue (AddComponent|NODE_AddHierarchicalInstancedStaticMeshComponent-0))
    (Components|StaticMesh|SetStaticMesh _returnvalue (Variables|Awning|GetInstancedMeshToSpawn))
    (Utilities|Array|Add (Variables|Awning|GetAR_InstancedMeshComponents) _returnvalue)
    (Components|InstancedStaticMesh|AddInstance _returnvalue (Math|Transform|MakeTransform _returnvalue_location (Math|Rotator|MakeRotator 0.0 0.0 (Math|Random|RandomFloatinRange (- (Variables|Awning|GetRandomRotatationAmount) (* (Variables|Awning|GetRandomRotatationAmount) 2.0)) (Variables|Awning|GetRandomRotatationAmount))) (Math|Vector|MakeVector (Variables|Awning|GetInstancedMeshScale) (Variables|Awning|GetInstancedMeshScale) (Variables|Awning|GetInstancedMeshScale))) false)))

(fn BuildExposedAssetProviderVersion(Implementation)
  (params (Version :type Fort%20Exposed%20Asset%20Provider%20Version%20Structure%20%28by%20ref%29))
  (|CallParentFunction Version)
  (ExposedAssetProvider|AddVersion Version 1)
  (return))

(fn CollectAssetReferencestoAllow(Implementation)
  (params (Collector :type Fort%20Exposed%20Asset%20Reference%20Collector%20Structure%20%28by%20ref%29))
  (|CallParentFunction Collector)
  (ExposedAssetProvider|CollectObjectPropertyAssetReferences Collector 0 (Variables|Awning|GetInstMeshList))
  (ExposedAssetProvider|CollectObjectPropertyAssetReferences Collector 0 (Variables|Awning|GetAR_MaterialList))
  (return))

(fn CycleMaterial
  (if (Utilities|Array|IsValidIndex (Variables|Awning|GetAR_MaterialList) (Variables|Awning|GetMatIndex))
    (Variables|Awning|SetMatIndex (+ (Variables|Awning|GetMatIndex) 1))
    (for _array_element (Variables|Awning|GetAR_InstancedMeshComponents)
      (Rendering|Material|SetMaterial _array_element 0 (Utilities|Array|Get(acopy) (Variables|Awning|GetAR_MaterialList) (Variables|Awning|GetMatIndex)))
      (if (not (Utilities|Array|IsValidIndex (Variables|Awning|GetAR_MaterialList) (Variables|Awning|GetMatIndex)))
        (Variables|Awning|SetMatIndex 0)))
    (else
      (Variables|Awning|SetMatIndex 0))))

(fn GetValidSockets
  (bind _output_get (LocalVariables|SetL_AR_AllSockets (Transformation|GetAllSocketNames (Variables|Debug|GetStaticMeshComponent))))
  (for _array_element _output_get
    (if (Utilities|String|Contains (Utilities|String|ToString(Name) _array_element) (Variables|Awning|GetSocketName))
      (Utilities|Array|AddUnique (Variables|Awning|GetAR_ValidSocketList) _array_element))))

(fn RunAddon
  (Awning|GetValidSockets)
  (for _array_element (Variables|Awning|GetAR_ValidSocketList)
    (Awning|AddNewInstancedMesh _array_element)))

(fn SwapMaterial
  (for _array_element (Variables|Awning|GetAR_InstancedMeshComponents)
    (Rendering|Material|SetMaterial _array_element 0 (Utilities|Array|Get(acopy) (Variables|Awning|GetAR_MaterialList) (Variables|Awning|GetMatIndex)))))

(fn ToggleAddonMeshes
  (if (Variables|Awning|GetSpawnExtraMeshes)
    (Variables|Awning|SetSpawnExtraMeshes false)
    (for _array_element (Variables|Awning|GetAR_InstancedMeshComponents)
      (Components|DestroyComponent _array_element))
    (else
      (Variables|Awning|SetSpawnExtraMeshes true)
      (Awning|RunAddon))))

(fn ConstructionScript
  (|CallParentFunction)
  (if (Variables|Awning|GetSpawnExtraMeshes)
    (Awning|GetValidSockets)
    (for _array_element (Variables|Awning|GetAR_ValidSocketList)
      (Awning|AddNewInstancedMesh _array_element)
      (Awning|SwapMaterial))))
